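# Niagara shockwave burst for the MaskMaker Smack effect, built from two active emitters: Sphere and Smoke.
# Sphere expands one mesh by User.Size_Scale while fading it through color and dynamic material curves.
# Smoke bursts torus-placed puffs plus a larger central puff, using spawn-group masks, drag, animated sub-images and dissolving alpha.
# Flare_Complex, GPU Sparks and LightDetail emitters are present but disabled.
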NiagaraSystem /PistolCanary/Assets/FX/Daigo/Smack/NS_PC_MaskMaker_Smack_Shockwave
dependencies {
 usedDataInterfaces [
  &NiagaraDataInterfaceArrayColor
  &NiagaraDataInterfaceCurve
  &NiagaraDataInterfaceVector2DCurve
  &NiagaraDataInterfaceColorCurve
  &NiagaraDataInterfaceArrayFloat3
 ]
 usedDynamicInputs []
 usedModules [&SystemState]
}
emitters {
 Flare_Complex {
  bEnabled=false
  emitterUpdateScript [
   &EmitterState()
   &SpawnBurst_Instantaneous(Use Spawn Probability=false, UseLoopCountLimit=false)
   &SpawnBurst_Instantaneous(Spawn Group=1, Use Spawn Probability=false, UseLoopCountLimit=false) {module=SpawnBurst_Instantaneous001}
   &SpawnBurst_Instantaneous(Spawn Group=2, Use Spawn Probability=false, UseLoopCountLimit=false) {module=SpawnBurst_Instantaneous002}
  ]
  particleSpawnScript [
   /SetVariables(Particles.VisibilityTag=&SpawnGroup()) {isSetParameters=true}
   &InitializeParticle(
    Color=&SelectLinearColorFromArray(
     Array Sampling Mode=::NewEnumerator1
     Direct Array Index=&SpawnGroup()
     Color Selection Array=DataInterface() {propertyValues {ColorData [<rgba 2 2 2 1>, <rgba 1 1 1 0.5>, <rgba 1 1 1 1>]}}
     Random Seed=?
    )
    Lifetime=&MaskFloatBySpawnGroup(Masked Float=0.3, Passthrough Float=0.3)
    Position Offset=<0 0 15>
    Sprite Rotation Angle=&MaskFloatBySpawnGroup(Masked Float=0, Passthrough Float=&RandomRangeFloat(Minimum=0, Maximum=1))
    Sprite Rotation Mode=::NewEnumerator1
    Sprite Size=&MaskVector2dBySpawnGroup(
     Masked Vector2D=<3000 100>
     Passthrough Vector2D=&Vector2DFromFloat(Value=&Multiply_Float(A=2000, B=$User.Size_Scale))
     Spawn Group Mask=0
    )
    Sprite Size Mode=::NewEnumerator4
    UsePositionOffset=true
    Write Material Random=true
   )
   &CameraOffset(Camera Offset Amount=&MaskFloatBySpawnGroup(Masked Float=50, Passthrough Float=80))
  ]
  particleUpdateScript [
   &ParticleState()
   &SpriteRotationRate(
    Delta Time:NiagaraFloat=&FloatFromCurve(FloatCurve=DataInterface() {propertyValues {Curve {keys [
         {tangentMode=RCTM_User, time=0, value=1}
         {arriveTangent=-0.3174, interpMode=RCIM_Cubic, leaveTangent=-0.3174, tangentMode=RCTM_User, time=0.3, value=1}
         {arriveTangent=0.0123, interpMode=RCIM_Cubic, leaveTangent=0.0123, tangentMode=RCTM_User, time=1, value=0}
        ]}}})
    Rotation Rate=&MaskFloatBySpawnGroup(Masked Float=5, Passthrough Float=0, Spawn Group Mask=2)
   )
   &ScaleSpriteSize(
    Scale Factor=&MaskVector2dBySpawnGroup(
     Masked Vector2D=&Vector2DFromCurve(
      Vector2Curve=DataInterface() {propertyValues {
        XCurve {keys [<at 0 1 RCIM_Cubic>, {arriveTangent=-3.1442, interpMode=RCIM_Cubic, leaveTangent=-3.1442, tangentMode=RCTM_User, time=1, value=0}]}
        YCurve {keys [
          {arriveTangent=4.8145, interpMode=RCIM_Cubic, time=0, value=1}
          {arriveTangent=-3.0695, interpMode=RCIM_Cubic, leaveTangent=-3.0695, tangentMode=RCTM_User, time=1, value=0}
         ]}
       }}
      ScaleCurve=<1 1>
     )
     Passthrough Vector2D=&Vector2DFromFloat(Value=&MaskFloatBySpawnGroup(
       Masked Float=&FloatFromCurve(FloatCurve=DataInterface() {propertyValues {Curve {keys [
            {arriveTangent=5.1291, interpMode=RCIM_Cubic, leaveTangent=5.1291, tangentMode=RCTM_User, time=0, value=0}
            {arriveTangentWeight=0.1274, interpMode=RCIM_Cubic, leaveTangentWeight=0.7778, tangentMode=RCTM_User, time=0.5, value=1}
           ]}}})
       Passthrough Float=&FloatFromCurve(FloatCurve=DataInterface() {propertyValues {Curve {keys [
            {arriveTangent=3.3451, interpMode=RCIM_Cubic, leaveTangent=3.3451, tangentMode=RCTM_User, time=0, value=0.5}
            {arriveTangentWeight=0.1274, interpMode=RCIM_Cubic, leaveTangentWeight=0.7778, tangentMode=RCTM_User, time=0.5, value=1}
           ]}}})
       Spawn Group Mask=1
      ))
     Spawn Group Mask=0
    )
    Scale Sprite Size Mode=::NewEnumerator2
   )
   &ScaleColor(
    Scale Alpha=&MaskFloatBySpawnGroup(
     Masked Float=1
     Passthrough Float=&MaskFloatBySpawnGroup(
      Masked Float=&FloatFromCurve(FloatCurve=DataInterface() {propertyValues {Curve {keys [<at 0 1>,<at 1 0>]}}}, Scale Curve=0.5)
      Passthrough Float=&FloatFromCurve(
       FloatCurve=DataInterface() {propertyValues {Curve {keys [
           {arriveTangent=-23.168, interpMode=RCIM_Cubic, leaveTangent=-23.168, tangentMode=RCTM_User, time=0, value=4}
           {arriveTangent=-4.5789, interpMode=RCIM_Cubic, leaveTangent=-4.5789, tangentMode=RCTM_User, time=0.2, value=1}
           <at 1 0 RCIM_Cubic>
          ]}}}
       Scale Curve=0.2
      )
      Spawn Group Mask=1
     )
    )
    Scale RGB=&MakeVectorFromLinearColorRGB(Color=&ColorFromCurve(ColorCurve=DataInterface() {propertyValues {
        AlphaCurve {keys []}
        BlueCurve {keys [{arriveTangent=-1.5872, interpMode=RCIM_Cubic, leaveTangent=-1.5872, tangentMode=RCTM_User, time=0, value=3}, <at 1 0.5 RCIM_Cubic>]}
        GreenCurve {keys [{arriveTangent=-4.9882, interpMode=RCIM_Cubic, leaveTangent=-4.9882, tangentMode=RCTM_User, time=0, value=3.75}, <at 1 0.625 RCIM_Cubic>]}
        RedCurve {keys [{arriveTangent=-3.5785, interpMode=RCIM_Cubic, leaveTangent=-3.5785, tangentMode=RCTM_User, time=0, value=4}, <at 1 1 RCIM_Cubic>]}
       }}))
    ScaleA=true
    ScaleRGB=true
   )
  ]
  rendererClasses [&NiagaraSpriteRendererProperties]
  renderers [{rendererClass=&NiagaraSpriteRendererProperties}]
 }
 LightDetail {
  bEnabled=false
  emitterUpdateScript [
   &EmitterState(Life Cycle Mode=::NewEnumerator1, Loop Behavior=::NewEnumerator1, Loop Duration=0.5, UseLoopDelay=false)
   &SpawnBurst_Instantaneous(Use Spawn Probability=false, UseLoopCountLimit=false)
  ]
  particleSpawnScript [&InitializeParticle(Color=<rgba 100 14.8142 11.9489 1>, Lifetime=0.5, UsePositionOffset=false, Write Material Random=true)]
  particleUpdateScript [
   &ParticleState()
   &ScaleColor(
    Scale Alpha=&FloatFromCurve(FloatCurve=DataInterface() {propertyValues {Curve {keys [{arriveTangent=-0.0581, interpMode=RCIM_Cubic, leaveTangent=-0.0581, tangentMode=RCTM_User, time=0, value=1}, <at 1 0>]}}})
    ScaleA=true
    ScaleRGB=true
   )
  ]
  rendererClasses [&NiagaraLightRendererProperties]
  renderers [{rendererClass=&NiagaraLightRendererProperties, rendererIndex=0}]
 }
 Smoke {
  emitterUpdateScript [
   &EmitterState(Life Cycle Mode=::NewEnumerator1, Loop Behavior=::NewEnumerator1, Loop Duration=1, UseLoopDelay=true)
   &SpawnBurst_Instantaneous(Spawn Count=5, Use Spawn Probability=false, UseLoopCountLimit=true)
   &SpawnBurst_Instantaneous(Spawn Group=1, Use Spawn Probability=false, UseLoopCountLimit=false) {module=SpawnBurst_Instantaneous001}
  ]
  particleSpawnScript [
   &InitializeParticle(
    Color Maximum=<rgba 0.3854 0.3854 0.3854 0.4957>
    Color Minimum=<rgba 0.2552 0.2552 0.2552 0.1>
    Color Mode=::NewEnumerator2
    Lifetime Max=1.5
    Lifetime Min=0.5
    Lifetime Mode=::NewEnumerator1
    Sprite Rotation Mode=::NewEnumerator2
    Sprite Size=&MaskVector2dBySpawnGroup(
     Masked Vector2D=&RandomRangeVector2D(Minimum=<247.401 233.8594>, Maximum=<309.1217 334.8896>)
     Passthrough Vector2D=&RandomRangeVector2D(Minimum=<1000 1000>, Maximum=<1220.4313 1321.5734>)
     Spawn Group Mask=0
    )
    Sprite Size Mode=::NewEnumerator4
    Sprite UV Mode=::NewEnumerator3
    UsePositionOffset=false
    Write Material Random=true
   )
   &CameraOffset(Camera Offset Amount=50)
   &ShapeLocation(
    Handle Radius=5
    Large Radius=277.0131
    MaskSpawnGroup=true
    Rotation Mode=::NewEnumerator1
    Shape Primitive=::NewEnumerator3
    Torus Distribution Mode:ENiagaraTorusDistributionMode=::NewEnumerator1
    U Position=&ReturnNormalizedExecIndex()
    UseOwnerScale=true
   )
   &ShapeLocation(MaskSpawnGroup=true, Override Seed=false, Spawn Group Mask=1, Sphere Radius=25, UseOwnerScale=true) {module=ShapeLocation001}
   &AddVelocity(
    Module.Use Falloff Distance=false
    Override Seed=false
    Velocity Mode=::NewEnumerator1
    Velocity Speed=&RandomRangeFloat(Minimum=1495.1179, Maximum=1519.1489)
   )
  ]
  particleUpdateScript [
   &Drag(Drag=&RandomRangeFloat(Minimum=6, Maximum=11), Use Linear Drag=true, Use Rotational Drag=true)
   &AccelerationForce(Acceleration=&RandomRangeVector(Minimum=<0 0 -100>, Maximum=<0 0 200>))
   &SolveForcesAndVelocity(Clamp Velocity=false, Limit Acceleration=false)
   &ParticleState()
   &ScaleSpriteSize(
    Uniform Curve Scale=4
    Uniform Curve Sprite Scale=DataInterface() {propertyValues {Curve {keys [
        {arriveTangent=4.1286, interpMode=RCIM_Cubic, leaveTangent=4.1286, tangentMode=RCTM_User, time=0, value=0.475}
        {arriveTangent=1.2171, interpMode=RCIM_Cubic, leaveTangent=1.2171, tangentMode=RCTM_User, time=0.3, value=1.125}
        <at 1 1.4 RCIM_Cubic>
       ]}}}
   )
   /SetVariables(Particles.SubImageIndex=&FloatFromCurve(
     FloatCurve=DataInterface() {propertyValues {Curve {keys [
         {arriveTangent=2.5597, interpMode=RCIM_Cubic, leaveTangent=2.5597, tangentMode=RCTM_User, time=0, value=0}
         {arriveTangent=1, interpMode=RCIM_Cubic, leaveTangent=1, time=0.1, value=0.75}
         {interpMode=RCIM_Cubic, leaveTangent=1.1425, time=1, value=1}
        ]}}}
     Scale Curve=63
    )) {isSetParameters=true}
   &DynamicMaterialParameters(
    Index 0 Param 1=&FloatFromCurve(FloatCurve=DataInterface() {propertyValues {Curve {keys [{arriveTangent=3.9017, interpMode=RCIM_Cubic, leaveTangent=3.9017, tangentMode=RCTM_User, time=0, value=0.075}, <at 0.5 1 RCIM_Cubic>]}}})
    Index 0 Param 4=1
    Module.Index0Float0Write=true
    Module.Index0Float1Write=true
    Module.Index0Float2Write=true
    Module.Index0Float3Write=true
   )
   &ScaleColor(
    Scale Alpha=&MaskFloatBySpawnGroup(
     Masked Float=&FloatFromCurve(FloatCurve=DataInterface() {propertyValues {Curve {keys [{arriveTangent=6.7279, interpMode=RCIM_Cubic, time=0, value=1}, {interpMode=RCIM_Cubic, leaveTangent=-0.9746, time=1, value=0}]}}})
     Passthrough Float=&FloatFromCurve(FloatCurve=DataInterface() {propertyValues {Curve {keys [
          {arriveTangent=-2.0221, interpMode=RCIM_Cubic, leaveTangent=-2.0221, tangentMode=RCTM_User, time=0, value=1}
          {interpMode=RCIM_Cubic, leaveTangent=-0.9746, time=1, value=0}
         ]}}})
    )
    Scale RGB=&MakeVectorFromLinearColorRGB(Color=&ColorFromCurve(ColorCurve=DataInterface() {propertyValues {
        AlphaCurve {keys []}
        BlueCurve {keys [{arriveTangent=-9.5363, interpMode=RCIM_Cubic, leaveTangent=-9.5363, tangentMode=RCTM_User, time=0, value=2.9711}, <at 0.127 1 RCIM_Cubic>]}
        GreenCurve {keys [{arriveTangent=-9.5363, interpMode=RCIM_Cubic, leaveTangent=-9.5363, tangentMode=RCTM_User, time=0, value=2.9688}, <at 0.127 1 RCIM_Cubic>]}
        RedCurve {keys [{arriveTangent=-9.5363, interpMode=RCIM_Cubic, leaveTangent=-9.5363, tangentMode=RCTM_User, time=0, value=3}, <at 0.127 1 RCIM_Cubic>]}
       }}))
    ScaleA=true
    ScaleRGB=true
   )
  ]
  rendererClasses [&NiagaraSpriteRendererProperties]
  renderers [{rendererClass=&NiagaraSpriteRendererProperties}]
 }
 Sparks {
  bEnabled=false
  simTarget=GPUComputeSim
  emitterUpdateScript [
   &EmitterState(Life Cycle Mode=::NewEnumerator1, Loop Behavior=::NewEnumerator1, Loop Delay=0.15, Loop Duration=1, UseLoopDelay=false)
   &SpawnBurst_Instantaneous(Spawn Count=100, Use Spawn Probability=false, UseLoopCountLimit=false)
  ]
  particleSpawnScript [
   &InitializeParticle(
    Color Maximum=<rgba 0.3385 0.3385 0.3385 1>
    Color Minimum=<rgba 0.0677 0.0677 0.0677 1>
    Color Mode=::NewEnumerator2
    Lifetime=&RandomRangeFloat(Minimum=0.2, Maximum=0.4)
    Sprite Size Mode=::NewEnumerator3
    Uniform Sprite Size=&FloatFromCurve(
     FloatCurve=DataInterface() {propertyValues {Curve {keys [{arriveTangent=-2.2424, interpMode=RCIM_Cubic, leaveTangent=-2.2424, tangentMode=RCTM_User, time=0, value=5}, <at 1 0.85 RCIM_Cubic>]}}}
     CurveIndex=&RandomRangeFloat(Minimum=0, Maximum=1)
     Scale Curve=$User.Size_Scale
    )
    UsePositionOffset=false
    Write Material Random=true
   )
   &ShapeLocation(
    Cylinder Height=0
    Cylinder Height Midpoint=0
    Cylinder Radius=&UniformAOrBFloat(A=199, B=400, Distribution Weight=0.412, Evaluation Type=::NewEnumerator0, Recalculate Random Each Loop=true, Override Seed=false)
    MaskSpawnGroup=false
    Override Seed=false
    Rotation Mode=::NewEnumerator2
    Shape Primitive=::NewEnumerator1
    Surface Only=true
    Use Endcaps In Surface Only Mode=false
    UseOwnerScale=true
    Yaw / Pitch / Roll=&MakeCustomVectorFromBool(
     Bool=&SetBoolByFloatComparison(A=&RandomRangeFloat(Minimum=0, Maximum=1), B=0.8)
     True Vector=&RandomVector(Vector Scale=1)
     False Vector=&SelectVectorFromArray(
      Array Sampling Mode=::NewEnumerator0
      Direct Array Index=?
      Vector Selection Array=DataInterface() {propertyValues {FloatData [<0.2252 0.236 0.2072>, <0.25 0.1351 0.3081>, <0.1081 0.2 0.2523>, <0 0.4595 0>]}}
      Random Seed=0
     )
    )
   )
   &AddVelocity(
    Velocity=$StackContext.ShapeLocation.ShapeNormal
    Velocity Speed Scale=&RandomRangeFloat(Minimum=5000, Maximum=20000)
   )
  ]
  particleUpdateScript [
   &ParticleState()
   &CurlNoiseForce(
    Bias Noise=false
    Noise Frequency=1000
    Noise Strength=&FloatFromCurve(
     FloatCurve=DataInterface() {propertyValues {Curve {
        defaultValue=1
        keys [
         {arriveTangent=-3.2557, interpMode=RCIM_Cubic, leaveTangent=-3.2557, tangentMode=RCTM_User, time=0, value=1}
         {arriveTangent=-0.7893, interpMode=RCIM_Cubic, leaveTangent=-0.7893, tangentMode=RCTM_User, time=0.1, value=0.25}
         {interpMode=RCIM_Cubic, leaveTangent=-8.0089, time=1, value=0}
        ]
       }}}
     Scale Curve=100000
    )
    Pan Noise=false
   )
   &Drag(Drag=5, Use Linear Drag=true, Use Rotational Drag=true)
   &PointForce(Force Strength=-200, Module.Use Falloff=false, Module.UseLinearForce=false)
   &SolveForcesAndVelocity(Clamp Velocity=false, Limit Acceleration=false)
   &ScaleSpriteSizeBySpeed(Max Scale Factor=<2 50>, Velocity Threshold=2000)
   &ScaleColor(
    Scale Alpha=&FloatFromCurve(FloatCurve=DataInterface() {propertyValues {Curve {keys [<at 0 1>,<at 1 0>]}}}, Scale Curve=0.5)
    ScaleA=true
    ScaleRGB=true
   )
  ]
  rendererClasses [&NiagaraSpriteRendererProperties]
  renderers [{rendererClass=&NiagaraSpriteRendererProperties}]
 }
 Sphere {
  emitterUpdateScript [
   &EmitterState()
   &SpawnBurst_Instantaneous(Use Spawn Probability=false, UseLoopCountLimit=false)
  ]
  particleSpawnScript [&InitializeParticle(
    Color Mode=::NewEnumerator0
    Lifetime=0.5
    Mesh Scale=<3 3 3>
    Mesh Scale Mode=::NewEnumerator4
    UsePositionOffset=false
    Write Material Random=true
   )]
  particleUpdateScript [
   &ParticleState()
   &ScaleMeshSize(Scale Factor=&VectorFromFloat(Value=&FloatFromCurve(
      FloatCurve=DataInterface() {propertyValues {Curve {
         defaultValue=0
         keys [
          {arriveTangent=4.1785, interpMode=RCIM_Cubic, leaveTangent=4.1785, tangentMode=RCTM_User, time=0.1, value=0.5}
          {arriveTangent=1.1179, interpMode=RCIM_Cubic, leaveTangent=1.1179, tangentMode=RCTM_User, time=0.2, value=0.75}
          {arriveTangent=0.0224, interpMode=RCIM_Cubic, leaveTangent=0.0224, tangentMode=RCTM_User, time=1, value=1}
         ]
        }}}
      Scale Curve=$User.Size_Scale
     )))
   &DynamicMaterialParameters(
    Index 0 Param 4=&FloatFromCurve(FloatCurve=DataInterface() {propertyValues {Curve {keys [<at 0 1 RCIM_Cubic>,<at 1 0 RCIM_Cubic>]}}}, Scale Curve=0.1)
    Module.Index0Float0Write=true
    Module.Index0Float1Write=true
    Module.Index0Float2Write=true
    Module.Index0Float3Write=true
   )
   &ScaleColor(
    Scale Alpha=&FloatFromCurve(FloatCurve=DataInterface() {propertyValues {Curve {keys [<at 0 1>,<at 1 0>]}}}, Scale Curve=0.05)
    Scale RGB=&MakeVectorFromLinearColorRGB(Color=&ColorFromCurve(ColorCurve=DataInterface() {propertyValues {
        AlphaCurve {keys [<at 0 1 RCIM_Cubic>,<at 1 0 RCIM_Cubic>]}
        BlueCurve {keys [<at 0 5>,<at 0.9798 1>]}
        GreenCurve {keys [<at 0 5>,<at 0.9798 0.4257>]}
        RedCurve {keys [<at 0 5>,<at 0.9798 0.1712>]}
       }}))
    ScaleA=true
    ScaleRGB=true
   )
  ]
  rendererClasses [&NiagaraMeshRendererProperties]
  renderers [{rendererClass=&NiagaraMeshRendererProperties, rendererIndex=0}]
 }
}
summary {
 systemName=NS_PC_MaskMaker_Smack_Shockwave
 emitters [
  {bEnabled=false, emitterName=Flare_Complex, rendererClasses [&NiagaraSpriteRendererProperties]}
  {emitterName=Smoke, rendererClasses [&NiagaraSpriteRendererProperties]}
  {emitterName=Sphere, rendererClasses [&NiagaraMeshRendererProperties]}
  {bEnabled=false, emitterName=Sparks, simTarget=GPUComputeSim, rendererClasses [&NiagaraSpriteRendererProperties]}
  {bEnabled=false, emitterName=LightDetail, rendererClasses [&NiagaraLightRendererProperties]}
 ]
 userVariables [
  User.Color {type=LinearColor, description="", defaultValue {struct=LinearColor, value=<rgba 1.852 0 20 1>}}
  User.Size {type=NiagaraFloat, defaultValue=50, description=""}
  User.Size_Scale {type=NiagaraFloat, defaultValue=5, description=""}
  User.Timing_BurstDelay {type=NiagaraFloat, defaultValue=0, description=""}
 ]
}
